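# M_Wheel_CSO is an unlit material for the CSO cosmetic wheel, drawing an animated, glowing spiral orb.
# Time-driven rotation, panning and noise distort spiral textures into patterns and RGB masks.
# Three color channels use either paint color or phase-offset scrolling gradient samples, rainbow by default.
# Masked color blends with invertible luminance, a dot mask and edge glow form the emissive, passed through vehicle damage and de-rez functions.
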Material /VehicleCosmetics/Wheels/CSO/Materials/M_Wheel_CSO
graph {
 properties {
  bAutomaticallySetUsageInEditor=false
  bUseMaterialAttributes=true
  bUsedWithMeshDeformer=true
  bUsedWithSkeletalMesh=true
  shadingModel=MSM_Unlit
  usedShadingModels=Unlit
 }
 nodes [
  n1 = NamedRerouteUsage()
  MaterialFunctionCall(In=MaterialFunctionCall(Material=SetMaterialAttributes(Emissive Color=@n1) {attributeSetTypes [#1]}) {materialFunction=&MF_BaseVehicle_Damage}.Result) {materialFunction=&MF_VehicleDeRez_Transition}
  n2 = Time()
  n3 = ScalarParameter() {defaultValue=0.5, parameterName=SpiralTimeOffset}
  NamedRerouteDeclaration(None=Multiply(A=Add(A=@n2, B=@n3), B=ScalarParameter() {defaultValue=1, parameterName=SpiralSpeed})) {name=Spiral Time, nodeColor=<rgba 1 0.5625 0 1>}
  n4 = TextureCoordinate()
  n5 = Multiply(A=Time()) {constB=0.2}
  n6 = Multiply(
   A=TextureSample(UVs=Multiply(A=@n4)) {samplerSource=SSM_Wrap_WorldGroupSettings, texture=/VehicleCosmetics/Wheels/_SharedTextures/FX/CloudNoise_01_uncompressed}.R
   B=Multiply(A=Sine(None=@n5)) {constB=0.03}
  )
  n7 = Rotator(
   Coordinate=Panner(Coordinate=TextureCoordinate() {uTiling=0.5, vTiling=0.5}, Time=Constant() {r=0.25}) {speedX=1, speedY=1}
   Time=NamedRerouteUsage()
  ) {speed=1}
  NamedRerouteDeclaration(None=TextureSample(UVs=Add(A=@n6, B=@n7)) {texture=/VehicleCosmetics/Wheels/CSO/Textures/Wheel_cso_Gradient_Spiral_02}.R) {name=Spiral Warp, nodeColor=<rgba 0.2422 0 1 1>}
  n8 = TextureSample() {samplerSource=SSM_Wrap_WorldGroupSettings, texture=/VehicleCosmetics/Wheels/_SharedTextures/FX/Universal_Noise_83_RGB}
  n9 = TextureSample!(UVs, Tex, Level, Apply View MipBias) {automaticViewMipBias=false, mipValueMode=TMVM_MipLevel, texture=/VehicleCosmetics/Wheels/CSO/Textures/Wheel_cso_Gradient01}
  n10 = Multiply(A=NamedRerouteUsage()) {constB=0.4}
  n11 = Panner(Coordinate=Add(A=AppendVector(A=@n8.R, B=@n9.R), B=@n10), Time=NamedRerouteUsage()) {speedX=-0.1, speedY=-0.1}
  n12 = NamedRerouteDeclaration(None=TextureObjectParameter() {parameterName=Spiral ORB Grid, texture=/VehicleCosmetics/Wheels/CSO/Textures/Wheel_SpiralOrb_RGB_Grid}) {name=Spiral ORB Grid, nodeColor=<rgba 0 1 0.5547 1>}
  n13 = Add(A=Multiply(A=@n10), B=Panner(Coordinate=TextureCoordinate() {coordinateIndex=1, uTiling=0.25}, Time=NamedRerouteUsage()) {speedX=-0.2})
  NamedRerouteDeclaration(None=Add(
    A=TextureSample!(UVs=@n11, Tex=@n12, Apply View MipBias) {samplerSource=SSM_Wrap_WorldGroupSettings, texture=&Wheel_SpiralOrb_RGB_Grid}.R
    B=TextureSample(UVs=@n13, Tex=@n12) {samplerSource=SSM_Wrap_WorldGroupSettings, texture=&Wheel_SpiralOrb_RGB_Grid}.G
   )) {name=Spiral Pattern, nodeColor=<rgba 0.0781 1 0 1>}
  n14 = TextureSample(UVs=@n11) {samplerSource=SSM_Wrap_WorldGroupSettings, texture=/VehicleCosmetics/Wheels/CSO/Textures/Wheel_SpiralOrb_Noise_13_RGB}
  NamedRerouteDeclaration(None=@n14.R) {name=Mask_R_Col, nodeColor=<rgba 1 0 0.3984 1>}
  NamedRerouteDeclaration(None=@n14.G) {name=Mask_G_Col, nodeColor=<rgba 1 0 0.3984 1>}
  NamedRerouteDeclaration(None=@n14.B) {name=Mask_B_Col, nodeColor=<rgba 1 0 0.3984 1>}
  n15 = VectorParameter() {parameterName=Paint_color, defaultValue=<rgba 1 1 1 1>}
  n16 = Multiply(A=TextureCoordinate()) {constB=0.1}
  n17 = Multiply(A=Time(), B=ScalarParameter() {defaultValue=0.1, parameterName=ColorPanSpeed})
  n18 = Panner(Coordinate=@n16, Time=@n17) {speedX=1}
  n19 = TextureObjectParameter() {parameterName=Gradient, texture=/VehicleCosmetics/Wheels/_SharedTextures/FX/Curve_Rainbow}
  NamedRerouteDeclaration(None=StaticSwitchParameter(
    True=@n15.RGB
    False=TextureSample(UVs=@n18, Tex=@n19) {samplerSource=SSM_Wrap_WorldGroupSettings, texture=&Curve_Rainbow}.RGB
   ) {parameterName=Painted}) {name=Color_1, nodeColor=<rgba 1 0 0.3984 1>}
  n20 = Add(A=@n17) {constB=0.25}
  n21 = Panner(Coordinate=@n16, Time=@n20) {speedX=1}
  NamedRerouteDeclaration(None=StaticSwitchParameter(
    True=@n15.RGB
    False=TextureSample(UVs=@n21, Tex=@n19) {samplerSource=SSM_Wrap_WorldGroupSettings, texture=&Curve_Rainbow}.RGB
   ) {parameterName=Painted}) {name=Color_2, nodeColor=<rgba 1 0 0.3984 1>}
  n22 = Panner(Coordinate=@n16, Time=Add(A=@n20) {constB=0.25}) {speedX=1}
  NamedRerouteDeclaration(None=StaticSwitchParameter(
    True=@n15.RGB
    False=TextureSample(UVs=@n22, Tex=@n19) {samplerSource=SSM_Wrap_WorldGroupSettings, texture=&Curve_Rainbow}.RGB
   ) {parameterName=Painted}) {name=Color_3, nodeColor=<rgba 1 0 0.3984 1>}
  n23 = ScalarParameter() {defaultValue=0.9, parameterName=EmissiveMultiply}
  n24 = Multiply(A=NamedRerouteUsage(), B=NamedRerouteUsage())
  n25 = Multiply(A=NamedRerouteUsage(), B=NamedRerouteUsage())
  n26 = Add(A=@n24, B=@n25)
  n27 = Multiply(A=NamedRerouteUsage(), B=NamedRerouteUsage())
  n28 = Multiply(A=@n24, B=@n25)
  n29 = Add(A=Add(A=Multiply(A=@n26, B=@n27), B=@n28), B=Add(A=@n28, B=Multiply(A=@n25, B=@n27)))
  n30 = Clamp(None=ConstantBiasScale(None=Desaturation(None=@n29) {luminanceFactors=<rgba 0.3 0.59 0.11 0>}) {bias=-0.1, scale=3})
  n31 = OneMinus(None=@n30)
  n32 = ScalarParameter() {parameterName=InvertSpiralLuminance}
  n33 = Clamp(None=ConstantBiasScale(None=TextureSample() {texture=/VehicleCosmetics/Wheels/_SharedTextures/FX/Dot_inverted}.R) {bias=-0.2, scale=1.8})
  n34 = Multiply(A=Add(A=LinearInterpolate(A=@n30, B=@n31, Alpha=@n32)), B=Multiply(A=@n29, B=@n33))
  n35 = NamedRerouteUsage()
  n36 = Multiply(A=LinearInterpolate(A=@n35, B=OneMinus(None=@n35), Alpha=@n32), B=@n33)
  n37 = NamedRerouteUsage()
  n38 = TextureSample(Tex=NamedRerouteUsage()) {constCoordinate=1, samplerSource=SSM_Wrap_WorldGroupSettings, texture=&Wheel_SpiralOrb_RGB_Grid}
  n39 = Add(
   A=Multiply(A=@n23, B=Multiply(A=@n24, B=@n36))
   B=Multiply(A=Multiply(A=@n37, B=@n38.B), B=ScalarParameter() {defaultValue=0.9, parameterName=EmissiveMultiply_Edge})
  )
  NamedRerouteDeclaration(None=Add(A=Multiply(A=@n23, B=@n34), B=Clamp(None=@n39))) {name=Emissive, nodeColor=<rgba 0.875 0 1 1>}
 ]
}
